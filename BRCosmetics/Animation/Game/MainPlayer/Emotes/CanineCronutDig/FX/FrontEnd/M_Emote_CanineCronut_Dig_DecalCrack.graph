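# Translucent lit Niagara sprite and mesh-particle material for the ground crack in the front-end CanineCronut Dig emote.
# Crack texture RGB is tinted by particle color, while a powered Height texture times LavaColor makes a glow term divided by pre-exposure.
# Dynamic-parameter ErosionProgress drives a soft smoothstep dissolve on texture red, invertible via FlipErosion, masking texture alpha.
# UseRedAsAlpha bypasses that erosion for raw red, before particle-alpha and opacity scaling.
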Material /BRCosmetics/Animation/Game/MainPlayer/Emotes/CanineCronutDig/FX/FrontEnd/M_Emote_CanineCronut_Dig_DecalCrack
graph {
 properties {
  bAutomaticallySetUsageInEditor=false
  bUsedWithNiagaraMeshParticles=true
  bUsedWithNiagaraSprites=true
  blendMode=BLEND_Translucent
  usedShadingModels=Default Lit
 }
 nodes [
  Constant() {r=1}
  Constant()
  n1 = Power(
   Base=TextureSampleParameter2D() {parameterName=Height, texture=/BRCosmetics/Animation/Game/MainPlayer/Emotes/CanineCronutDig/FX/T_Emote_CanineCronutDig_Crack_Flip}.RGB
   Exp=ScalarParameter() {defaultValue=1, parameterName=LavaPower}
  )
  n2 = VectorParameter() {parameterName=LavaColor, defaultValue=<rgba 1 0.2614 0 1>}
  Divide(
   A=Multiply(A=Multiply(A=@n1, B=@n2.RGB), B=ScalarParameter() {defaultValue=1, parameterName=FE_Emissive_Modifier})
   B=ViewProperty() {property=MEVP_PreExposure}.Property
  )
  n3 = TextureSampleParameter2D() {
   parameterName=Base
   samplerType=SAMPLERTYPE_LinearColor
   texture=/BRCosmetics/Animation/Game/MainPlayer/Emotes/CanineCronutDig/FX/T_Emote_CanineCronutDig_Crack
  }
  Multiply(A=ParticleColor().RGB, B=@n3.RGB)
  n4 = TextureSampleParameter2D() {
   parameterName=Base
   samplerType=SAMPLERTYPE_LinearColor
   texture=&T_Emote_CanineCronutDig_Crack
  }
  n5 = DynamicParameter() {defaultValue=<rgba 0 0 1 1>, paramNames [ErosionProgress, FlipErosion, Param3, Param4]}
  n6 = ScalarParameter() {defaultValue=0.1, parameterName=Falloff_Erosion}
  n7 = MaterialFunctionCall(
   Input=Add(A=ScalarParameter() {defaultValue=0.1, parameterName=Progress_Erosion_offset}, B=@n5.ErosionProgress)
   Input Low=Constant()
   Input High=Constant() {r=1}
   Target Low=Subtract(B=@n6) {constA=0}
   Target High=Constant() {r=1}
  ) {materialFunction=&RemapValueRange}
  n8 = TextureSampleParameter2D() {
   parameterName=Base
   samplerType=SAMPLERTYPE_LinearColor
   texture=&T_Emote_CanineCronutDig_Crack
  }
  n9 = OneMinus(None=@n8.R)
  n10 = Multiply(A=@n4.A, B=SmoothStep(Min=@n7.Result, Max=Add(A=@n7.Result, B=@n6), Value=If(A=@n5.FlipErosion, "A > B"=@n9, "A == B"=@n8.R, "A < B"=@n8.R)))
  n11 = StaticBoolParameter() {parameterName=UseRedAsAlpha}
  Multiply(
   A=Multiply(A=StaticSwitch(True=@n4.R, False=@n10, Value=@n11), B=ParticleColor().A)
   B=ScalarParameter() {defaultValue=1, parameterName=Parameter_Opacity}
  )
 ]
}
